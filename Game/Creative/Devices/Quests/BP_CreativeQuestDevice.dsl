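; Creative quest tracker device that tracks a stat toward a target for assigned players, then broadcasts completion and can end the round or unassign.
; Knobs copy into a CreativeQuest component; minigame state changes and trigger events drive assignment, progress, target and resets.
; Optional persistence saves and loads active state, progress and target per player, with a configurable first/average/median/max/min rule resolving shared-progress conflicts.
; Assign requests are batched for a frame to avoid duplicate triggers.
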(event EventBeginPlay
  (Default|UpdateGameState)
  (Persistence|InitializePersistenceComponent (Variables|Default|GetFortCreativePersistenceComponent) (Variables|Default|GetFortMinigameProgress) true)
  (bind _returnvalue (Game|ConstructObjectfromClass (Variables|Persistence|GetPersistenceDataObjectClass)))
  (Variables|Persistence|SetPersistenceDataObject _returnvalue))

(event OnAnyPropertyChanged(ToyOptionsComponent)
  ;[0.1624 1 0.8499] Initialize Device
  (CallFunction|CopyProperties (Variables|Default|GetCreativeQuest)))

(event OnPropertiesLoaded(ToyOptionsComponent)
  ;[0.1624 1 0.8499] Initialize Device
  (CallFunction|CopyProperties (Variables|Default|GetCreativeQuest)))

(event MinigameStateChangedDelegate(FortMinigameProgress)
  (params (Minigame :type Fort%20Minigame%20Object%20Reference) (NewMinigameState :type EFortMinigameState%20Enum :default "PreGame"))
  (Default|UpdateGameState)
  (switch Utilities|FlowControl|Switch|SwitchonEFortMinigameState NewMinigameState
    (:Setup
      (Networking|SwitchHasAuthority
        (:Authority
          (if (== (Variables|/Script/FortniteGame.FortMinigame|GetCurrentRound Minigame) 1)
            (Class|CreativeQuestComponent|ResetQuestProgress (Variables|Default|GetCreativeQuest) (Variables|Knob|GetInitialAmount))
            (Default|SetMinigameTargetAmount (Variables|Knob|GetTargetAmount))
            (Default|Unassign 0)))))
    (:InProgress
      (Networking|CanExecuteCosmeticEvents
        (:True
          (Utilities|FlowControl|Delay 0.0)
          (Class|CreativeQuestComponent|TriggerHUDUpdate (Variables|Default|GetCreativeQuest) true))))))

(event Custom|Increment%20Target%20Amount%20Event
  (params (Instigator :type Controller%20Object%20Reference))
  ;[0.2703 0 1] Gameplay Functions
  (Utilities|QuestIsActive Instigator)
  (Default|IncrementMinigameTargetAmount 1 Instigator))

(event Custom|Decrement%20Target%20Amount%20Event
  (params (Instigator :type Controller%20Object%20Reference))
  ;[0.2703 0 1] Gameplay Functions
  (Utilities|QuestIsActive Instigator)
  (Default|IncrementMinigameTargetAmount -1 Instigator))

(event OnConditionsMetDelegate(CreativeQuest)
  (params (ConditionInstigator :type Controller%20Object%20Reference) (IncrementAmount :type Integer))
  (Default|IncrementCurrentAmount ConditionInstigator IncrementAmount true))

(event Custom|Increment%20Progress%20Event
  (params (Instigator :type Controller%20Object%20Reference))
  ;[0.2703 0 1] Gameplay Functions
  (Default|IncrementCurrentAmount Instigator (Variables|Knob|GetIncrementCurrentByOnTrigger)))

(event Custom|OnPawnJoinInProgress
  (params (Pawn :type Pawn%20Object%20Reference))
  (bind _returnvalue (Pawn|GetController Pawn))
  (Utilities|IsValid Pawn
    (:Is%20Valid
      (Utilities|IsValid _returnvalue
        (:Is%20Valid
          (if (Variables|Knob|GetKnob_AssignAtJoinInProgress)
            (CallFunction|AssignforPlayer _returnvalue)
            (Utilities|FlowControl|Sequence
              (:then_0
                _)
              (:then_1
                (CallFunction|ResetDistanceTimer)))
            (else
              _))
          (if (and (Variables|UserOptions|Basic|GetUsePersistence) (Variables|Knob|GetbAutoLoad))
            (Persistence|LoadProgress _returnvalue true)))))))

(event Custom|Decrement%20Progress%20Event
  (params (Instigator :type Controller%20Object%20Reference))
  ;[0.2703 0 1] Gameplay Functions
  (Default|IncrementCurrentAmount Instigator (* (Variables|Knob|GetIncrementCurrentByOnTrigger) -1)))

(event MinigameStartedDelegate(FortMinigameProgress)
  (if (Networking|HasAuthority)
    (Utilities|FlowControl|Sequence
      (:then_0
        (CallFunction|BindUnbindtoJoininProgressEvent true))
      (:then_1
        (if (Variables|Knob|GetKnob_AssignOnStart)
          (CallFunction|AssignforPlayer 0)))
      (:then_2
        (Class|CreativeQuestComponent|OnMinigameStarted (Variables|Default|GetCreativeQuest)))
      (:then_3
        (if (and (Variables|UserOptions|Basic|GetUsePersistence) (Variables|Knob|GetbAutoLoad))
          (Persistence|LoadProgressforAll)))
      (:then_4
        (Persistence|SetAutoSaveEnabled (Variables|Default|GetFortCreativePersistenceComponent) (and (Variables|UserOptions|Basic|GetUsePersistence) (Variables|Knob|GetbAutoSave)))))))

(event Custom|Assign%20Tracker%20Event
  (params (Instigator :type Controller%20Object%20Reference))
  ;[0.2703 0 1] Gameplay Functions
  (Utilities|IsValid Instigator
    (:Is%20Valid
      (Utilities|Set|Add (Variables|Default|GetAssignReceiverInstigators) Instigator)
      ;[1 1 1] Delay to next frame to allow this device to handle other events before this one as this can cause duplicate event triggers)
      (Utilities|FlowControl|Delay 0.0)
      (bind _result (Utilities|Set|ToArray (Variables|Default|GetAssignReceiverInstigators)))
      (for _array_element _result
        (CallFunction|AssignforPlayer _array_element))
      (Utilities|Set|Clear (Variables|Default|GetAssignReceiverInstigators)))))

(event Custom|Assign%20Tracker%20To%20All%20Event
  (params (Instigator :type Controller%20Object%20Reference))
  ;[0.2703 0 1] Gameplay Functions
  (CallFunction|AssignforPlayer 0))

(event Custom|Remove%20Tracker%20Event
  (params (Instigator :type Controller%20Object%20Reference))
  ;[0.2703 0 1] Gameplay Functions
  (Utilities|IsValid Instigator
    (:Is%20Valid
      (Default|Unassign Instigator))))

(event EventPostFinishSpawning
  (|CallParentFunction)
  (if (not (Variables|Default|GetMigratedPersistence))
    (Variables|UserOptions|Basic|SetUsePersistence (Math|Boolean|ORBoolean (Class|FortGameplayMessageComponentBase|IsChannelIdSet (Variables|Default|GetSaveProgressReceiver)) (Class|FortGameplayMessageComponentBase|IsChannelIdSet (Variables|Default|GetLoadProgressForPlayerReceiver)) (Class|FortGameplayMessageComponentBase|IsChannelIdSet (Variables|Default|GetLoadProgressForAllReceiver)) (Class|FortGameplayMessageComponentBase|IsChannelIdSet (Variables|Default|GetClearPersistenceReceiver)) (Variables|Knob|GetbAutoLoad) (Variables|Knob|GetbAutoSave)))
    (Variables|Default|SetMigratedPersistence true)
    (OptionData|RefreshPropertyOverrides (Variables|Default|GetToyOptionsComponent))
    (Class|FortCreativePersistenceComponentBase|UpdateIsPersistencePosible (Variables|Default|GetFortCreativePersistenceComponent))))

(event Custom|Remove%20Tracker%20From%20All%20Event
  (params (Instigator :type Controller%20Object%20Reference))
  ;[0.2703 0 1] Gameplay Functions
  (Default|Unassign 0))

(event MinigameEndedDelegate(FortMinigameProgress)
  (if (Networking|HasAuthority)
    (Utilities|FlowControl|Sequence
      (:then_0
        (if (and (not (Math|Boolean|ORBoolean (Utilities|Enum|Equal(Enum) (Utilities|Enum|BytetoEnumECreativeQuestStat (Math|Conversions|ToByte(Integer) (Variables|Knob|GetStatToTrack))) "RoundCompleted") (Utilities|Enum|Equal(Enum) (Utilities|Enum|BytetoEnumECreativeQuestStat (Math|Conversions|ToByte(Integer) (Variables|Knob|GetStatToTrack))) "RoundWon") (Utilities|Enum|Equal(Enum) (Utilities|Enum|BytetoEnumECreativeQuestStat (Math|Conversions|ToByte(Integer) (Variables|Knob|GetStatToTrack))) "GameCompleted") (Utilities|Enum|Equal(Enum) (Utilities|Enum|BytetoEnumECreativeQuestStat (Math|Conversions|ToByte(Integer) (Variables|Knob|GetStatToTrack))) "GameWon"))) (Math|Conversions|ToBoolean(Integer) (Variables|Knob|GetResetBetweenRounds)))
          (Class|CreativeQuestComponent|ResetQuestProgress (Variables|Default|GetCreativeQuest) (Variables|Knob|GetInitialAmount))
          (Default|SetMinigameTargetAmount (Variables|Knob|GetTargetAmount))
          (Default|Unassign 0)))
      (:then_1
        (CallFunction|BindUnbindtoJoininProgressEvent false))
      (:then_2
        (Persistence|SetAutoSaveEnabled (Variables|Default|GetFortCreativePersistenceComponent) false))
      (:then_3
        (Class|CreativeQuestComponent|OnMinigameEnded (Variables|Default|GetCreativeQuest))))))

(event Custom|Complete%20Tracker%20Event
  (params (Instigator :type Controller%20Object%20Reference))
  ;[0.2703 0 1] Gameplay Functions
  (Utilities|QuestIsActive Instigator)
  (Default|CheckTargetAmountReached Instigator true))

(event Custom|Reset%20Tracker%20Event
  (params (Instigator :type Controller%20Object%20Reference))
  ;[0.2703 0 1] Gameplay Functions
  (Default|SetCurrentAmount Instigator (Variables|Knob|GetInitialAmount)))

(event Custom|Save%20Tracker%20Event
  (params (Instigator :type Controller%20Object%20Reference))
  ;[0.2703 0 1] Gameplay Functions
  (if (Variables|UserOptions|Basic|GetUsePersistence)
    (Persistence|SaveProgress Instigator)))

(event Custom|Load%20Tracker%20Event
  (params (Instigator :type Controller%20Object%20Reference))
  ;[0.2703 0 1] Gameplay Functions
  (if (Variables|UserOptions|Basic|GetUsePersistence)
    (Persistence|LoadProgress Instigator)))

(event EventOnWorldReady
  (Visibility|SetVisibilityComponents (Variables|Default|GetVisibleInGameComponent) (Utilities|Array|MakeArray (Variables|Debug|GetStaticMeshComponent) (Variables|Default|GetSM_QuestDevice_Attachment) (Variables|Default|Getw1) (Variables|Default|GetSM_BDA_Attachment_05) (Variables|Default|Getw) (Variables|Default|GetSM_BDA_Attachment_03) (Variables|Default|GetSM_BDA_Attachment_04)) true))

(event Custom|Load%20Tracker%20For%20All%20Event
  (params (Instigator :type Controller%20Object%20Reference))
  ;[0.2703 0 1] Gameplay Functions
  (if (Variables|UserOptions|Basic|GetUsePersistence)
    (Persistence|LoadProgressforAll)))

(event Custom|Clear%20Persistence%20Event
  (params (Instigator :type Controller%20Object%20Reference))
  ;[0.2703 0 1] Gameplay Functions
  (if (Variables|UserOptions|Basic|GetUsePersistence)
    (Persistence|ClearPersistenceforPlayer Instigator)))

(fn ApplyLoadedData
  (params (bActive :type Boolean) (Progress :type Integer) (Target :type Integer) (Player :type Controller%20Object%20Reference))
  (Class|CreativeQuestComponent|SetActiveFor (Variables|Default|GetCreativeQuest) (Variables|GetPlayer) (Variables|GetbActive))
  (Default|SetCurrentAmount (Variables|GetPlayer) (Variables|GetProgress))
  (Default|SetMinigameTargetAmount (Variables|GetTarget) (Variables|GetPlayer)))

(fn AssignForPlayer
  (params (Instigator :type Controller%20Object%20Reference))
  (Class|CreativeQuestComponent|SetActiveFor (Variables|Default|GetCreativeQuest) Instigator true)
  (Default|CheckTargetAmountReached Instigator))

(fn BindUnbindToJoinInProgressEvent
  (params (Bind :type Boolean))
  (bind _outputdelegate (EventDispatchers|CreateEvent :!event OnPawnJoinInProgress))
  (bind _fortminigameprogress (Variables|Default|GetFortMinigameProgress))
  (bind _currentminigame (Variables|/Script/FortniteGame.FortMinigameLogicComponent|GetCurrentMinigame _fortminigameprogress))
  (Utilities|IsValid _currentminigame
    (:Is%20Valid
      (if (Variables|GetBind)
        (Minigame|BindEventtoOnPlayerJoininProgress _currentminigame _outputdelegate)
        (else
          (Minigame|UnbindEventfromOnPlayerJoininProgress _currentminigame _outputdelegate))))))

(fn CanPlayerAutoSave
  (params (FortPlayerState :type Fort%20Player%20State%20Object%20Reference))
  (return true))

(fn CheckTargetAmountReached
  (params (Instigator :type Controller%20Object%20Reference) (bForceComplete :type Boolean))
  (bind _instigator (Variables|Actor|GetInstigator))
  (bind _creativequest (Variables|Default|GetCreativeQuest))
  (bind _returnvalue (Class|CreativeQuestComponent|GetActiveSharedControllers _creativequest _instigator))
  (bind _instigator_2 (Variables|Actor|GetInstigator))
  (bind (_winning_team _team_index) (CallFunction|WinningTeamType _instigator_2))
  (bind _fortminigameprogress (Variables|Default|GetFortMinigameProgress))
  (bind _currentminigame (Variables|/Script/FortniteGame.FortMinigameLogicComponent|GetCurrentMinigame _fortminigameprogress))
  (if (or (Variables|GetbForceComplete) (and (>= (Default|GetCurrentAmount Instigator) (Variables|Default|GetMinigameTargetAmount)) (> (Variables|Default|GetMinigameTargetAmount) 0)))
    (switch int (Variables|Knob|GetTargetReachedBehavior)
      (:0
        (Utilities|IsValid _currentminigame
          (:Is%20Valid
            (if (Minigame|IsinProgress _currentminigame)
              (Utilities|FlowControl|Sequence
                (:then_0
                  (if (Utilities|Enum|NotEqual(Enum) _winning_team "NewEnumerator0")
                    (Minigame|SetWinningTeamIndex _currentminigame (Minigame|GetWinningTeamIndexfromTeamIndex _currentminigame _team_index))))
                (:then_1
                  (Minigame|EndGame _currentminigame "EndRound")))))))
      (:1
        _)
      (:2
        (for _array_element _returnvalue
          (Default|Unassign _array_element))))
    (for _array_element_2 _returnvalue
      (GameplayEvent|BroadcastEvent (Variables|UserOptions|Events|GetOn%20Complete) 0 (Utilities|Struct|MakeGameplayEventQuestComplete _array_element_2)))))

(fn ClearPersistenceForPlayer
  (params (Instigator :type Controller%20Object%20Reference))
  (bind _playerstate (Variables|/Script/Engine.Controller|GetPlayerState Instigator))
  (Utilities|IsValid Instigator
    (:Is%20Valid
      (Utilities|IsValid _playerstate
        (:Is%20Valid
          (Persistence|ClearData (Variables|Default|GetFortCreativePersistenceComponent) (PlayerState|GetUniqueNetId _playerstate) (Variables|Persistence|GetSavePath)))))))

(fn CopyProperties
  (params (Component :type Creative%20Quest%20Component%20Object%20Reference))
  (Utilities|IsValid Component
    (:Is%20Valid
      (Variables|/Script/FortniteGame.CreativeQuestComponent|SetTargetTeam (Variables|Knob|GetTeamSetting) (Variables|GetComponent))
      (Variables|/Script/FortniteGame.CreativeQuestComponent|SetTargetClass (Variables|Knob|GetClassSetting) (Variables|GetComponent))
      (Variables|/Script/FortniteGame.CreativeQuestComponent|SetTargetDeathCause (Utilities|Select :Index (Variables|Knob|GetWeaponCategory) :Option%200 "Unspecified" :Option%201 "FallDamage" :Option%202 "Pistol" :Option%203 "SMG" :Option%204 "Shotgun" :Option%205 "Rifle" :Option%206 "Sniper" :Option%207 "GrenadeLauncher" :Option%208 "RocketLauncher" :Option%209 "Bow" :Option%2010 "Melee") (Variables|GetComponent))
      (Variables|/Script/FortniteGame.CreativeQuestComponent|SetStatToTrack (Utilities|Enum|BytetoEnumECreativeQuestStat (Math|Conversions|ToByte(Integer) (Variables|Knob|GetStatToTrack))) (Variables|GetComponent))
      (Variables|Default|SetMinigameTargetAmount (Variables|Knob|GetTargetAmount))
      (Variables|/Script/FortniteGame.CreativeQuestComponent|SetbSelfEliminationsCount (Variables|Knob|GetSelfEliminationsCount) (Variables|GetComponent))
      (Variables|/Script/FortniteGame.CreativeQuestComponent|SetSharingMode (Variables|Knob|GetSharingMode) (Variables|GetComponent))
      (Variables|/Script/FortniteGame.CreativeQuestComponent|SetAssignedToTeam (Variables|Knob|GetAssignedToTeam) (Variables|GetComponent))
      (Variables|/Script/FortniteGame.CreativeQuestComponent|SetQuestName (Variables|Knob|GetQuestName) (Variables|GetComponent))
      (Variables|/Script/FortniteGame.CreativeQuestComponent|SetObjectiveText (Variables|Knob|GetQuestObjectiveText) (Variables|GetComponent))
      (Variables|/Script/FortniteGame.CreativeQuestComponent|SetMinigameTargetAmount (Variables|Default|GetMinigameTargetAmount) (Variables|GetComponent))
      (Variables|/Script/FortniteGame.CreativeQuestComponent|SetShowProgressMode (Utilities|Enum|BytetoEnumEShowProgressMode (Math|Conversions|ToByte(Integer) (Variables|Knob|GetShowQuestProgress))) (Variables|GetComponent))
      (Variables|/Script/FortniteGame.CreativeQuestComponent|SetShowQuestOnHUD (Variables|Knob|GetShowOnHUD) (Variables|GetComponent))
      (Variables|/Script/FortniteGame.CreativeQuestComponent|SetbShowCompleteEffects (Variables|Knob|GetShowCompleteEffects) (Variables|GetComponent))
      (Variables|/Script/FortniteGame.CreativeQuestComponent|SetbHeadshotsOnly (Math|Conversions|ToBoolean(Integer) (Variables|Knob|GetHeadshotsOnly)) (Variables|GetComponent))
      (Variables|/Script/FortniteGame.CreativeQuestComponent|SetbPlayerEliminationsCount (Math|Conversions|ToBoolean(Integer) (Variables|Knob|GetPlayerEliminationsCount)) (Variables|GetComponent))
      (Variables|/Script/FortniteGame.CreativeQuestComponent|SetbNonPlayerEliminationsCount (Math|Conversions|ToBoolean(Integer) (Variables|Knob|GetNonPlayerEliminationsCount)) (Variables|GetComponent))
      (Variables|/Script/FortniteGame.CreativeQuestComponent|SetQuestIconImage (Creative|UI|IconPicker|GetIconLibraryTexture (Variables|Knob|GetQuestIcon)) (Variables|GetComponent))
      (Variables|/Script/FortniteGame.CreativeQuestComponent|SetQuestColor (Variables|Knob|GetColor) (Variables|GetComponent))
      (Networking|CanExecuteCosmeticEvents
        (:True
          (Class|CreativeQuestComponent|OnRepQuestUpdate (Variables|GetComponent)))))))

(fn GetFirstPlayerProgress
  (params (Instigator :type Controller%20Object%20Reference))
  (Utilities|FlowControl|Sequence
    (:then_0
      (LocalVariables|SetSharedControllers (Class|CreativeQuestComponent|GetActiveSharedControllers (Variables|Default|GetCreativeQuest) (Variables|Actor|GetInstigator))))
    (:then_1
      (for _array_element (Minigame|GetPlayersOrdered (Variables|/Script/FortniteGame.FortMinigameLogicComponent|GetCurrentMinigame (Variables|Default|GetFortMinigameProgress)))
        (if (Utilities|Array|ContainsItem (LocalVariables|GetSharedControllers) (PlayerState|GetPlayerController _array_element))
          (bind (_bloadsuccess _loadeddata) (Persistence|LoadToPlayer (PlayerState|GetPlayerController _array_element)))
          (if _bloadsuccess
            (return (Variables|/Game/Creative/Devices/Quests/PersistenceData_QuestDevice.PersistenceData_QuestDevice_C|GetProgress _loadeddata) (Variables|/Game/Creative/Devices/Quests/PersistenceData_QuestDevice.PersistenceData_QuestDevice_C|GetTarget _loadeddata)))))
      (return 0 (Variables|Knob|GetTargetAmount)))))

(fn GetCurrentAmount
  (params (Controller :type Controller%20Object%20Reference))
  (return (Class|CreativeQuestComponent|GetCurrentAmount (Variables|Default|GetCreativeQuest) Controller)))

(fn GetObjectToAutoSaveForPlayer
  (params (FortPlayerState :type Fort%20Player%20State%20Object%20Reference))
  (Persistence|PopulatePersistenceDataObject (PlayerState|GetPlayerController FortPlayerState))
  (return (Variables|Persistence|GetPersistenceDataObject)))

(fn GetPersistenceRequiredState
  (return (Variables|UserOptions|Basic|GetUsePersistence)))

(fn IncrementCurrentAmount
  (params (Controller :type Controller%20Object%20Reference) (IncrementBy :type Integer) (BypassActiveCheck :type Boolean))
  (if (not BypassActiveCheck)
    (Utilities|QuestIsActive Controller))
  (Default|SetCurrentAmount (Variables|GetController) (Math|Integer|Max(Integer) (+ (Default|GetCurrentAmount Controller) IncrementBy))))

(fn IncrementMinigameTargetAmount
  (params (IncrementBy :type Integer :default 1) (Instigator :type Controller%20Object%20Reference))
  (Default|SetMinigameTargetAmount (+ (Variables|Default|GetMinigameTargetAmount) IncrementBy) Instigator))

(fn LoadForSharedControllersAndDisambiguate
  (params (Instigator :type Controller%20Object%20Reference))
  (switch Utilities|FlowControl|Switch|SwitchonECreativeQuest_ResolveConflicts (Variables|Knob|GetConflictResolutionMethod)
    (:NewEnumerator0
      (bind (_progress _target) (Persistence|GetFirstPlayerProgress (Variables|Actor|GetInstigator)))
      (return _progress _target))
    (:NewEnumerator1
      (bind (_outprogresslist _outtargetlist) (Persistence|LoadProgressfromSharedControllers (Variables|Actor|GetInstigator)))
      (return (Math|Float|Round (Math|Integer|AverageOfIntArray _outprogresslist)) (Math|Float|Round (Math|Integer|AverageOfIntArray _outtargetlist))))
    (:NewEnumerator2
      (bind (_outprogresslist_2 _outtargetlist_2) (Persistence|LoadProgressfromSharedControllers (Variables|Actor|GetInstigator)))
      (return (Math|Float|Round (Math|Integer|MedianOfIntArray _outprogresslist_2)) (Math|Float|Round (Math|Integer|MedianOfIntArray _outtargetlist_2))))
    (:NewEnumerator3
      (bind (_outprogresslist_3 _outtargetlist_3) (Persistence|LoadProgressfromSharedControllers (Variables|Actor|GetInstigator)))
      (return (.MaxValue (Math|Integer|MaxOfIntArray _outprogresslist_3)) (.MaxValue (Math|Integer|MaxOfIntArray _outtargetlist_3))))
    (:NewEnumerator4
      (bind (_outprogresslist_4 _outtargetlist_4) (Persistence|LoadProgressfromSharedControllers (Variables|Actor|GetInstigator)))
      (return (.MinValue (Math|Integer|MinOfIntArray _outprogresslist_4)) (.MinValue (Math|Integer|MinOfIntArray _outtargetlist_4))))))

(fn LoadProgress
  (params (Instigator :type Controller%20Object%20Reference) (bIsJIP :type Boolean))
  (bind _instigator (Variables|Actor|GetInstigator))
  (bind _creativequest (Variables|Default|GetCreativeQuest))
  (bind _returnvalue (Class|CreativeQuestComponent|GetActiveSharedControllers _creativequest _instigator))
  (Utilities|IsValid Instigator
    (:Is%20Valid
      (switch Utilities|FlowControl|Switch|SwitchonECreativeQuestSharing (Variables|Knob|GetSharingMode)
        (:Individual
          (bind (_bloadsuccess _loadeddata) (Persistence|LoadToPlayer (Variables|Actor|GetInstigator)))
          (if _bloadsuccess
            (Persistence|ApplyLoadedData (Variables|/Game/Creative/Devices/Quests/PersistenceData_QuestDevice.PersistenceData_QuestDevice_C|GetbActive _loadeddata) (Variables|/Game/Creative/Devices/Quests/PersistenceData_QuestDevice.PersistenceData_QuestDevice_C|GetProgress _loadeddata) (Variables|/Game/Creative/Devices/Quests/PersistenceData_QuestDevice.PersistenceData_QuestDevice_C|GetTarget _loadeddata) (Variables|Actor|GetInstigator))))
        (:Team
          _)
        (:All
          _))
      (if (not (Variables|GetbIsJIP))
        _
        (elif (Variables|Knob|GetbResolveAfterActive)
          _
          (elif (Utilities|Array|IsNotEmpty _returnvalue)
            (Class|CreativeQuestComponent|SetCurrentAmount (Variables|Default|GetCreativeQuest) (Variables|Actor|GetInstigator) (Class|CreativeQuestComponent|GetCurrentAmount (Variables|Default|GetCreativeQuest) (Utilities|Array|Get(acopy) _returnvalue)))
            (else
              (bind (_bloadsuccess_2 _loadeddata_2) (Persistence|LoadToPlayer (Variables|Actor|GetInstigator)))
              (if _bloadsuccess_2
                (Persistence|ApplyLoadedData (Variables|/Game/Creative/Devices/Quests/PersistenceData_QuestDevice.PersistenceData_QuestDevice_C|GetbActive _loadeddata_2) (Variables|/Game/Creative/Devices/Quests/PersistenceData_QuestDevice.PersistenceData_QuestDevice_C|GetProgress _loadeddata_2) (Variables|/Game/Creative/Devices/Quests/PersistenceData_QuestDevice.PersistenceData_QuestDevice_C|GetTarget _loadeddata_2) (Variables|Actor|GetInstigator)))))))
      (bind (_progress _target) (Persistence|LoadforSharedControllersandDisambiguate (Variables|Actor|GetInstigator)))
      (for _array_element (Class|CreativeQuestComponent|GetActiveSharedControllers (Variables|Default|GetCreativeQuest) (Variables|Actor|GetInstigator))
        (Utilities|IsValid _array_element
          (:Is%20Valid
            (bind (_bloadsuccess_3 _loadeddata_3) (Persistence|LoadToPlayer _array_element))
            (Persistence|ApplyLoadedData (select _bloadsuccess_3 (Variables|/Game/Creative/Devices/Quests/PersistenceData_QuestDevice.PersistenceData_QuestDevice_C|GetbActive _loadeddata_3) (Class|CreativeQuestComponent|IsActiveFor (Variables|Default|GetCreativeQuest) _array_element)) _progress _target _array_element)))))))

(fn LoadProgressForAll
  (bind _fortminigameprogress (Variables|Default|GetFortMinigameProgress))
  (bind _currentminigame (Variables|/Script/FortniteGame.FortMinigameLogicComponent|GetCurrentMinigame _fortminigameprogress))
  (Utilities|IsValid _currentminigame
    (:Is%20Valid
      (bind _persistencemanagercomponent (Variables|/Script/FortniteGame.FortMinigame|GetPersistenceManagerComponent _currentminigame)
        (:then
          (LocalVariables|SetPersistencePlayers (Persistence|GetPersistencePlayers _persistencemanagercomponent :!class /Script/FortniteGame.FNEPersistenceManagerComponent))
          (switch Utilities|FlowControl|Switch|SwitchonECreativeQuestSharing (Variables|Knob|GetSharingMode)
            (:Individual
              (for _array_element (LocalVariables|GetPersistencePlayers)
                (bind _output_get (LocalVariables|SetCurrentPlayer (PlayerState|GetPlayerController _array_element)))
                (bind (_bloadsuccess _loadeddata) (Persistence|LoadToPlayer _output_get))
                (if _bloadsuccess
                  (Persistence|ApplyLoadedData (Variables|/Game/Creative/Devices/Quests/PersistenceData_QuestDevice.PersistenceData_QuestDevice_C|GetbActive _loadeddata) (Variables|/Game/Creative/Devices/Quests/PersistenceData_QuestDevice.PersistenceData_QuestDevice_C|GetProgress _loadeddata) (Variables|/Game/Creative/Devices/Quests/PersistenceData_QuestDevice.PersistenceData_QuestDevice_C|GetTarget _loadeddata) (LocalVariables|GetCurrentPlayer)))))
            (:Team
              (for _array_element_2 (LocalVariables|GetPersistencePlayers)
                (LocalVariables|SetCurrentPlayer (PlayerState|GetPlayerController _array_element_2))
                (if (not (Utilities|Array|ContainsItem (LocalVariables|GetAlreadyLoadedPlayers) (LocalVariables|GetCurrentPlayer)))
                  (bind (_progress _target) (Persistence|LoadforSharedControllersandDisambiguate (LocalVariables|GetCurrentPlayer)))
                  (for _array_element_3 (Class|CreativeQuestComponent|GetActiveSharedControllers (Variables|Default|GetCreativeQuest) (LocalVariables|GetCurrentPlayer))
                    (Utilities|IsValid _array_element_3
                      (:Is%20Valid
                        (bind (_bloadsuccess_2 _loadeddata_2) (Persistence|LoadToPlayer _array_element_3))
                        (Persistence|ApplyLoadedData (select _bloadsuccess_2 (Variables|/Game/Creative/Devices/Quests/PersistenceData_QuestDevice.PersistenceData_QuestDevice_C|GetbActive _loadeddata_2) (Class|CreativeQuestComponent|IsActiveFor (Variables|Default|GetCreativeQuest) _array_element_3)) _progress _target _array_element_3)
                        (Utilities|Array|Add (LocalVariables|GetAlreadyLoadedPlayers) _array_element_3)))))))
            (:All
              (bind _array_element_4 (Utilities|Array|ForEachLoopwithBreak (LocalVariables|GetPersistencePlayers))
                (:LoopBody
                  (if (CallFunction|PlayerTeamApplies _array_element_4)
                    (LocalVariables|SetCurrentPlayer (PlayerState|GetPlayerController _array_element_4))))
                (:Completed
                  _))
              (Utilities|IsValid (LocalVariables|GetCurrentPlayer)
                (:Is%20Valid
                  _)))))
        (:else)))))

(fn LoadProgressFromSharedControllers
  (params (Instigator :type Controller%20Object%20Reference))
  (for _array_element (Class|CreativeQuestComponent|GetActiveSharedControllers (Variables|Default|GetCreativeQuest) Instigator)
    (bind (_bloadsuccess _loadeddata) (Persistence|LoadToPlayer _array_element))
    (Utilities|Array|Add (LocalVariables|GetProgressList) (select _bloadsuccess (Variables|/Game/Creative/Devices/Quests/PersistenceData_QuestDevice.PersistenceData_QuestDevice_C|GetProgress _loadeddata) 0))
    (Utilities|Array|Add (LocalVariables|GetTargetList) (select _bloadsuccess (Variables|/Game/Creative/Devices/Quests/PersistenceData_QuestDevice.PersistenceData_QuestDevice_C|GetTarget _loadeddata) (Variables|Knob|GetTargetAmount))))
  (return (LocalVariables|GetProgressList) (LocalVariables|GetTargetList)))

(fn LoadtoPlayer
  (params (Controller :type Controller%20Object%20Reference))
  (bind _persistencedataobject (Variables|Persistence|GetPersistenceDataObject))
  (Utilities|IsValid Controller
    (:Is%20Valid
      (Persistence|ResetPersistenceObject)
      (Persistence|LoadData (Variables|Default|GetFortCreativePersistenceComponent) (PlayerState|GetUniqueNetId (Variables|/Script/Engine.Controller|GetPlayerState Controller)) (Variables|Persistence|GetSavePath) _persistencedataobject)
      (if (Variables|/Game/Creative/Devices/Quests/PersistenceData_QuestDevice.PersistenceData_QuestDevice_C|GetbIsDataLoaded _persistencedataobject)
        (return true _persistencedataobject)
        (else
          (return false 0))))))

(fn PlayerTeamApplies
  (params (Actor :type Actor%20Object%20Reference))
  (bind (_teamindex _returnvalue) (Team|GetActorPvPTeamIndex Actor))
  (return (or (and _returnvalue (== _teamindex (Knobs|Team|GetTeamValueasIndex (Variables|Knob|GetAssignedToTeam)))) (Knobs|Team|IsTeamAny (Variables|Knob|GetAssignedToTeam)))))

(fn PopulatePersistenceDataObject
  (params (Controller :type Controller%20Object%20Reference))
  (Persistence|ResetPersistenceObject)
  (Utilities|IsValid Controller
    (:Is%20Valid
      (Variables|/Game/Creative/Devices/Quests/PersistenceData_QuestDevice.PersistenceData_QuestDevice_C|SetbActive (Class|CreativeQuestComponent|IsActiveFor (Variables|Default|GetCreativeQuest) (Variables|GetController)) (Variables|Persistence|GetPersistenceDataObject))
      (Variables|/Game/Creative/Devices/Quests/PersistenceData_QuestDevice.PersistenceData_QuestDevice_C|SetProgress (Default|GetCurrentAmount (Variables|GetController)) (Variables|Persistence|GetPersistenceDataObject))
      (Variables|/Game/Creative/Devices/Quests/PersistenceData_QuestDevice.PersistenceData_QuestDevice_C|SetTarget (Variables|Default|GetMinigameTargetAmount) (Variables|Persistence|GetPersistenceDataObject))
      (Variables|/Game/Creative/Devices/Quests/PersistenceData_QuestDevice.PersistenceData_QuestDevice_C|SetbIsDataLoaded true (Variables|Persistence|GetPersistenceDataObject)))))

(fn ResetPersistenceObject
  (bind _persistencedataobject (Variables|Persistence|GetPersistenceDataObject))
  (Variables|/Game/Creative/Devices/Quests/PersistenceData_QuestDevice.PersistenceData_QuestDevice_C|SetbActive false _persistencedataobject)
  (Variables|/Game/Creative/Devices/Quests/PersistenceData_QuestDevice.PersistenceData_QuestDevice_C|SetProgress 0 _persistencedataobject)
  (Variables|/Game/Creative/Devices/Quests/PersistenceData_QuestDevice.PersistenceData_QuestDevice_C|SetTarget 0 _persistencedataobject)
  (Variables|/Game/Creative/Devices/Quests/PersistenceData_QuestDevice.PersistenceData_QuestDevice_C|SetbIsDataLoaded false _persistencedataobject))

(fn ResetDistanceTimer
  (bind _creativequest (Variables|Default|GetCreativeQuest))
  (switch Utilities|FlowControl|Switch|SwitchonECreativeQuestStat (Utilities|Enum|BytetoEnumECreativeQuestStat (Math|Conversions|ToByte(Integer) (Variables|Knob|GetStatToTrack)))
    (:DistanceTraveledOnFoot
      _)
    (:DistanceTraveledInVehicle
      _)
    (:DistanceTraveledInAir
      _)
    (:DistanceTraveled
      _))
  (Class|CreativeQuestComponent|OnMinigameEnded _creativequest) ; Stop the timer if it's running
  (Class|CreativeQuestComponent|OnMinigameStarted _creativequest) ; Start the timer
)

(fn SaveForPlayer
  (params (Controller :type Controller%20Object%20Reference))
  (Utilities|IsValid Controller
    (:Is%20Valid
      (Persistence|PopulatePersistenceDataObject Controller)
      (Persistence|SaveData (Variables|Default|GetFortCreativePersistenceComponent) (Variables|Persistence|GetPersistenceDataObject) (PlayerState|GetUniqueNetId (Variables|/Script/Engine.Controller|GetPlayerState (Variables|GetController))) (Variables|Persistence|GetSavePath) false))))

(fn SaveProgress
  (params (Instigator :type Controller%20Object%20Reference%20%28by%20ref%29))
  (Utilities|IsValid Instigator
    (:Is%20Valid
      (switch Utilities|FlowControl|Switch|SwitchonECreativeQuestSharing (Variables|Knob|GetSharingMode)
        (:Individual
          (Persistence|SaveforPlayer (Variables|Actor|GetInstigator)))
        (:Team
          _)
        (:All
          _))
      (for _array_element (Class|CreativeQuestComponent|GetActiveSharedControllers (Variables|Default|GetCreativeQuest) (Variables|Actor|GetInstigator))
        (Persistence|SaveforPlayer _array_element)))))

(fn SetCurrentAmount
  (params (Controller :type Controller%20Object%20Reference) (Amount :type Integer))
  (Class|CreativeQuestComponent|SetCurrentAmount (Variables|Default|GetCreativeQuest) Controller Amount)
  (Default|CheckTargetAmountReached (Variables|GetController))
  ;[1 1 1] Jira FORT-450197 - Make sure the minigame stat component also updates when setting score.
  (if (and (Utilities|Enum|Equal(Enum) (Utilities|Enum|BytetoEnumECreativeQuestStat (Math|Conversions|ToByte(Integer) (Variables|Knob|GetStatToTrack))) "Score") (Variables|Default|GetbOverrideScoreboard))
    (bind _output_get (LocalVariables|SetMinigameStats_component (Minigame|Stats|GetMinigameStatsComponent self)))
    (Utilities|IsValid _output_get
      (:Is%20Valid
        (Minigame|Stats|SetStatforPlayerId (LocalVariables|GetMinigameStats_component) (PlayerState|GetUniqueNetId (Variables|/Script/Engine.Controller|GetPlayerState (Variables|GetController))) (Class|FortMinigameStatsComponent|GetLoadedStatfromClass (LocalVariables|GetMinigameStats_component) (LocalVariables|GetMinigameStat_score)) (Variables|GetAmount) false)))))

(fn SetMinigameTargetAmount
  (params (Amount :type Integer) (Instigator :type Controller%20Object%20Reference))
  (bind _output_get (Variables|Default|SetMinigameTargetAmount Amount))
  (Variables|/Script/FortniteGame.CreativeQuestComponent|SetMinigameTargetAmount _output_get (Variables|Default|GetCreativeQuest))
  ;[1 1 1] FORT-438964: Because we changed the target amount, check if any player who has this quest has completed it (not just the instigator)
  (bind _currentminigame (Variables|/Script/FortniteGame.FortMinigameLogicComponent|GetCurrentMinigame (Variables|Default|GetFortMinigameProgress))
    (:then
      (for _array_element (Minigame|GetParticipatingPlayers _currentminigame)
        (Utilities|QuestIsActive (PlayerState|GetPlayerController _array_element))
        (Default|CheckTargetAmountReached (PlayerState|GetPlayerController _array_element))))
    (:else)))

(fn Unassign
  (params (ForPlayer :type Controller%20Object%20Reference))
  (Class|CreativeQuestComponent|SetActiveFor (Variables|Default|GetCreativeQuest) ForPlayer false))

(fn UpdateGameState
  (bind _creativequest (Variables|Default|GetCreativeQuest))
  (bind _fortminigameprogress (Variables|Default|GetFortMinigameProgress)
    (:then
      (bind _currentminigame (Variables|/Script/FortniteGame.FortMinigameLogicComponent|GetCurrentMinigame _fortminigameprogress)
        (:then
          (if (Utilities|Enum|Equal(Enum) (Variables|/Script/FortniteGame.FortMinigame|GetCurrentState _currentminigame) "PreGame")
            (Class|CreativeQuestComponent|UnregisterQuest _creativequest)
            (else
              (Class|CreativeQuestComponent|RegisterQuest _creativequest _currentminigame))))
        (:else)))
    (:else)))

(fn ConstructionScript)

(fn WinningTeamType
  (params (Instigator :type Controller%20Object%20Reference))
  (Utilities|IsValid Instigator
    (:Is%20Valid
      (LocalVariables|SetInstigatorTeam (Team|GetActorPvPTeamIndex Instigator)))
    (:Is%20Not%20Valid
      _))
  (return (select (Knobs|Team|IsValidTeamIndex (Variables|Knob|GetWinningTeamOnEndGame)) "NewEnumerator2" (select (and (Utilities|IsValid (Variables|Actor|GetInstigator) :!class /Script/Engine.KismetSystemLibrary) (Knobs|Team|IsCompletingTeamWins (Variables|Knob|GetWinningTeamOnEndGame))) "NewEnumerator1" "NewEnumerator0")) (Utilities|Select :Index (select (Knobs|Team|IsValidTeamIndex (Variables|Knob|GetWinningTeamOnEndGame)) "NewEnumerator2" (select (and (Utilities|IsValid (Variables|Actor|GetInstigator) :!class /Script/Engine.KismetSystemLibrary) (Knobs|Team|IsCompletingTeamWins (Variables|Knob|GetWinningTeamOnEndGame))) "NewEnumerator1" "NewEnumerator0")) :NewEnumerator0 0 :NewEnumerator1 (LocalVariables|GetInstigatorTeam) :NewEnumerator2 (Math|Conversions|ToByte(Integer) (Knobs|Team|GetTeamValueasIndex (Variables|Knob|GetWinningTeamOnEndGame))))))
