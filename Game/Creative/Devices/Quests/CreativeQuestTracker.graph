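# HUD widget for Creative quest trackers that shows the active quest when its HUD mode is Detailed or Both.
# Asynchronously loads the quest's widget class into a SizeBox, repeating the load if the quest updates mid-load.
# Feeds that widget an MVVM tracker view model with name, description, color, async-loaded icon and progress.
# Plays Intro on show; when the quest clears, plays a completion or plain outro if complete effects are enabled, otherwise collapses.
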WidgetBlueprint /Game/Creative/Devices/Quests/CreativeQuestTracker
animations [
 Outro {
  binding_count=1
  length_seconds=0.25
  object=&self:Outro
  sequence {bindings [Overlay {
     id=#1
     possessed_class=Overlay
     tracks [
      {
       type=MovieSceneFloatTrack
       display_name=Render Opacity
       sections [{type=MovieSceneFloatSection, end_frame=5, start_frame=0, channels [None {type=MovieSceneScriptingFloatChannel, keys [[0,1],[5,0]]}]}]
      }
      {
       type=MovieSceneByteTrack
       display_name=Visibility
       sections [{type=MovieSceneByteSection, end_frame=5, start_frame=0, channels [None {type=MovieSceneScriptingByteChannel, keys [[0,4],[5,1]]}]}]
      }
     ]
    }]}
 }
 Intro {
  binding_count=1
  length_seconds=0.5
  object=&self:Intro
  sequence {
   bindings [Overlay {
     id=#2
     possessed_class=Overlay
     tracks [
      {
       type=MovieSceneFloatTrack
       display_name=Render Opacity
       sections [{type=MovieSceneFloatSection, channels [None {type=MovieSceneScriptingFloatChannel, keys [[0,0],[5,1],[10,1]]}]}]
      }
      {
       type=MovieScene2DTransformTrack
       sections [{
         type=MovieScene2DTransformSection
         channels [
          Translation.X {type=MovieSceneScriptingFloatChannel, keys [[0,-50],[10,0]]}
          Translation.Y {type=MovieSceneScriptingFloatChannel}
          Angle {type=MovieSceneScriptingFloatChannel}
          Scale.X {type=MovieSceneScriptingFloatChannel}
          Scale.Y {type=MovieSceneScriptingFloatChannel}
          Shear.X {type=MovieSceneScriptingFloatChannel}
          Shear.Y {type=MovieSceneScriptingFloatChannel}
         ]
        }]
      }
      {
       type=MovieSceneByteTrack
       display_name=Visibility
       sections [{type=MovieSceneByteSection, end_frame=0, start_frame=0, channels [None {type=MovieSceneScriptingByteChannel, const [3,1]}]}]
      }
     ]
    }]
   tracks [{
     type=MovieSceneEventTrack
     display_name=Event_StopCompleteAnimation
     sections [{
       type=MovieSceneEventTriggerSection
       channels [None {type=MovieSceneScriptingEventChannel, const ["<Struct 'MovieSceneEvent' (0x467bbda0000) {}>", 1]}]
      }]
    }]
  }
 }
 Outro_Complete {
  binding_count=1
  length_seconds=1.2
  object=&self:Outro_Complete
  sequence {bindings [Overlay {
     id=#1
     possessed_class=Overlay
     tracks [
      {
       type=MovieSceneFloatTrack
       display_name=Render Opacity
       sections [{
         type=MovieSceneFloatSection
         end_frame=23
         start_frame=-1
         channels [None {type=MovieSceneScriptingFloatChannel, keys [[-1,1],[17,1],[23,0]]}]
        }]
      }
      {
       type=MovieSceneByteTrack
       display_name=Visibility
       sections [{
         type=MovieSceneByteSection
         end_frame=23
         start_frame=-1
         channels [None {type=MovieSceneScriptingByteChannel, thinned_from=3, keys [[-1,4],[23,1]]}]
        }]
      }
     ]
    }]}
 }
]
graphs [
 Create and Set View Model {nodes [
   Create and Set View Model() {is=title K2Node_FunctionEntry}
   > Get!(execute:Exec) -> (then:Exec, else:Exec, LoadedWidget:User Widget Object Reference)
   > n1 = Get!(execute) -> (then, else, ViewModel:Device - Tracker View Model Object Reference)
   n2 = Get View From User Widget!(
    self:Viewmodel Engine Subsytem Object Reference=Get MVVMSubsystem() -> (ReturnValue:Viewmodel Engine Subsytem Object Reference) {is=title K2Node_GetEngineSubsystem}
    UserWidget:User Widget Object Reference=Get LoadedWidget() -> (LoadedWidget:User Widget Object Reference)
   ) -> (ReturnValue:MVVM View Object Reference) {is=title K2Node_CallFunction}
   Is Valid(exec=[@n1.then, @n5.then], InputObject=@n2)
   n3 = Get ViewModel() -> (ViewModel:Device - Tracker View Model Object Reference)
   > Set IsViewModelInitialized!(execute, IsViewModelInitialized:bool=false, self:Device - Tracker View Model Object Reference=@n3) -> (then, Output_Get:bool)
   > SetViewModel!(execute, self:MVVM View Object Reference=@n2, ViewModelName:Name=MVVM_UEFN_Tracker, ViewModel:Notify Field Value Changed Interface=@n3) -> (then, ReturnValue:bool)
   > Set IsViewModelInitialized(IsViewModelInitialized=true, self=@n3)
   n4 = Construct Object from Class(execute=@n1.else, Class=&/MVVM_UEFN_Tracker.MVVM_UEFN_Tracker_C) -> (ReturnValue:Device - Tracker View Model Object Reference)
   > n5 = Set ViewModel(ViewModel:Device - Tracker View Model Object Reference=@n4) -> (Output_Get:Device - Tracker View Model Object Reference)
  ]}
 EventGraph {nodes [
   Initialize ViewModel() {is=title K2Node_CustomEvent}
   > n1 = Get!(execute) -> (then, else, Cached Quest Component:Creative Quest Component Object Reference)
   > n2 = Branch(Condition=Get IsLoading() -> (IsLoading:bool))
   > Set WidgetUpdatedWhileLoading(WidgetUpdatedWhileLoading:bool=true) -> (Output_Get:bool)
   Set IsLoading(execute=@n2.else, IsLoading:bool=true) -> (Output_Get:bool)
   > n3 = Get!(execute) -> (then, else, LoadedWidget:User Widget Object Reference)
   > n4 = RemoveFromParent(self=@n3)
   n5 = CreateWidgetAsync!(
    execute:Exec=[@n4.then, @n3.else]
    UserWidgetSoftClass:User Widget Soft Class Reference=Get QuestWidget!(self:Creative Quest Component Object Reference=@n1) -> (QuestWidget:User Widget Soft Class Reference)
    OwningPlayer:Player Controller Object Reference
    bSuspendInputUntilComplete:bool=false
   ) -> (then:Exec, OnComplete:Exec, OnFailure:Exec, UserWidget:User Widget Object Reference) {is=title K2Node_AsyncAction}
   n6 = Branch(execute=@n5.OnComplete, Condition=Get WidgetUpdatedWhileLoading() -> (WidgetUpdatedWhileLoading:bool))
   > Set IsLoading(IsLoading=false)
   > Set WidgetUpdatedWhileLoading(WidgetUpdatedWhileLoading=false)
   > Initialize ViewModel() {is=title K2Node_CallFunction}
   n7 = Set LoadedWidget(execute=@n6.else, LoadedWidget:User Widget Object Reference=@n5) -> (Output_Get:User Widget Object Reference)
   > AddChild(self=Get SizeBox_Root() -> (SizeBox_Root:Size Box Object Reference), Content=@n7)
   > Set IsLoading(IsLoading=false)
   > Set WidgetUpdatedWhileLoading(WidgetUpdatedWhileLoading=false)
   > Create and Set View Model() {is=title K2Node_CallFunction}
   > Get!(execute) -> (then, else, Cached Quest Component:Creative Quest Component Object Reference)
   > n8 = Load Quest Icon!(execute:Exec) -> ("  ":Exec, Icon:Texture 2D Object Reference, Icon Visibility:ESlateVisibility Enum=Visible) {is=title K2Node_MacroInstance}
   > n9 = GetTrackerDetails!(execute, self) -> (
    then
    Name:Text
    Description:Text
    Color:Linear Color Structure
    Progress Visibility:ESlateVisibility Enum=Visible
    TargetValue:int=0
    ShowCompleteAnimation:bool
   )
   > UpdateStaticDetails!(
    execute
    self:Device - Tracker View Model Object Reference=Get ViewModel()
    Name:Text=@n9
    Description:Text=@n9.Description
    Color:Linear Color Structure=@n9.Color
    ProgressVisibility:ESlateVisibility Enum=@n9.Progress Visibility
    MaxValue:int=@n9.TargetValue
    ShowCompleteAnimation:bool=@n9.ShowCompleteAnimation
    Icon:Texture 2D Object Reference=@n8
    IconVisibility:ESlateVisibility Enum=@n8.Icon Visibility
   ) -> (then)
   > Update ViewModel Progress()
   > PlayAnimation(InAnimation=Get Intro())
   n10 = Event OnQuestChanged!() -> (OutputDelegate, then, CurrentQuestComponent:Creative Quest Component Object Reference, bHasActiveQuest:bool) {is=title K2Node_Event}
   > Set Cached Quest Component(Cached Quest Component:Creative Quest Component Object Reference=@n10.CurrentQuestComponent) -> (Output_Get:Creative Quest Component Object Reference)
   > n11 = Branch(Condition=@n10.bHasActiveQuest)
   > n12 = Get()
   > ShowNewQuest!(execute, self, QuestComponent:Creative Quest Component Object Reference=@n12) -> (then)
   > Set ShowCompleteAnimation(ShowCompleteAnimation:bool=Get bShowCompleteEffects!(self:Creative Quest Component Object Reference=@n12) -> (bShowCompleteEffects:bool)) -> (Output_Get:bool)
   n13 = Branch(execute=@n11.else, Condition=Get ShowCompleteAnimation() -> (ShowCompleteAnimation:bool))
   > Set ShowCompleteAnimation(ShowCompleteAnimation=false)
   > n14 = Branch(Condition=IsAnimationPlaying(InAnimation=Get Intro()))
   > n15 = SetAnimationCurrentTime!(execute, self:User Widget Object Reference, InAnimation:Widget Animation Object Reference=Get Intro(), InTime:float=0.5) -> (then)
   n16 = Branch(execute=[@n15.then, @n14.else], Condition=Get HasReachedCompletionAmount() -> (HasReachedCompletionAmount:bool))
   > PlayAnimation(InAnimation=Get Outro_Complete() -> (Outro_Complete:Widget Animation Object Reference))
   PlayAnimation(execute=@n16.else, InAnimation=Get Outro() -> (Outro:Widget Animation Object Reference))
   SetVisibility(execute=@n13.else, InVisibility=Collapsed)
   Event OnProgressUpdated() {is=title K2Node_Event}
   > Update ViewModel Progress()
   > n17 = Get()
   n18 = GetCurrentAmount!(self:Creative Quest Component Object Reference=@n17, ForPlayer:Controller Object Reference=GetOwningPlayer()) -> (ReturnValue:int=0)
   n19 = Get MinigameTargetAmount!(self:Creative Quest Component Object Reference=@n17) -> (MinigameTargetAmount:int=0)
   > Set HasReachedCompletionAmount(HasReachedCompletionAmount:bool=AND Boolean(A="integer <= integer"(A=@n19, B=@n18), B="integer > integer"(A=@n18))) -> (Output_Get:bool)
   n20 = Event OnDisplayMinigameInProgressUI!() -> (OutputDelegate, then, bDisplay:bool) {is=title K2Node_Event}
   > Parent: OnDisplayMinigameInProgressUI!(execute, bDisplay:bool=@n20.bDisplay) -> (then) {is=title K2Node_CallParentFunction}
   > n21 = Branch(Condition=@n20.bDisplay)
   SetVisibility(execute=@n21.else, InVisibility=Collapsed)
  ]}
 GetTrackerDetails {nodes [
   GetTrackerDetails() {is=title K2Node_FunctionEntry}
   n1 = Get Cached Quest Component() -> (Cached Quest Component:Creative Quest Component Object Reference)
   n2 = Get Cached Quest Component()
   > Return Node!(
    execute
    Name:Text="To Text (String)"(InString=Get QuestName!(self:Creative Quest Component Object Reference=@n1) -> (QuestName:String))
    Description:Text="To Text (String)"(InString=Get ObjectiveText!(self:Creative Quest Component Object Reference=@n1) -> (ObjectiveText:String))
    Color:Linear Color Structure=Get QuestColor!(self:Creative Quest Component Object Reference=Get Cached Quest Component()) -> (QuestColor:Linear Color Structure)
    Progress Visibility:ESlateVisibility Enum=Select!(
     Total:ESlateVisibility Enum=HitTestInvisible
     Remaining:ESlateVisibility Enum=HitTestInvisible
     Off:ESlateVisibility Enum=Collapsed
     Index:EShowProgressMode Enum=Get ShowProgressMode!(self:Creative Quest Component Object Reference=@n2) -> (ShowProgressMode:EShowProgressMode Enum=Total)
    ) -> (ReturnValue)
    TargetValue:int=Get MinigameTargetAmount(self=Get Cached Quest Component()) -> (MinigameTargetAmount=0)
    ShowCompleteAnimation:bool=Get bShowCompleteEffects(self=Get Cached Quest Component())
   )
  ]}
 LoadQuestIcon {nodes [
   Inputs()
   > n1 = Get!(execute, self:Creative Quest Component Object Reference=Get Cached Quest Component()) -> (then, else, QuestIconImage:Texture 2D Soft Object Reference)
   > n2 = Async Load Asset(Asset=@n1)
   n3 = Cast To Texture2D(execute=@n2.Completed, Object=@n2)
   Outputs!(
    "  ":Exec=[@n3.then, @n1.else]
    Icon:Texture 2D Object Reference=@n3
    Icon Visibility:ESlateVisibility Enum=Select!(
     Option 0:ESlateVisibility Enum=Collapsed
     Option 1:ESlateVisibility Enum=HitTestInvisible
     Index:bool=IsValidSoftObjectReference(SoftObjectReference=@n1)
    ) -> (ReturnValue)
   )
  ]}
 Sequencer Events {nodes [
   Stop Complete Animation() {is=title K2Node_CustomEvent}
   > StopAnimation(InAnimation=Get Outro())
  ]}
 ShowNewQuest {nodes [
   n1 = ShowNewQuest!() -> (then, QuestComponent:Creative Quest Component Object Reference) {is=title K2Node_FunctionEntry}
   > n2 = Switch on EShowOnHudMode!(
    execute
    Selection:EShowOnHudMode Enum=Get ShowQuestOnHUD!(self:Creative Quest Component Object Reference=@n1) -> (ShowQuestOnHUD:EShowOnHudMode Enum=No)
   ) -> (No, Detailed:Exec, List:Exec, Both:Exec) {is=title K2Node_SwitchEnum}
   Stop Complete Animation(execute=[@n2.Both, @n2.Detailed]) {is=title K2Node_CallFunction}
   > SetVisibility(InVisibility=SelfHitTestInvisible)
   > Initialize ViewModel()
   SetVisibility(execute=[@n2.List, @n2.No], InVisibility=Collapsed)
  ]}
 Update ViewModel Progress {nodes [
   Update ViewModel Progress() {is=title K2Node_FunctionEntry}
   > n1 = Get!(execute) -> (then, else, Cached Quest Component:Creative Quest Component Object Reference)
   n2 = Get MinigameTargetAmount(self=@n1) -> (MinigameTargetAmount=0)
   n3 = GetCurrentAmount(self=@n1, ForPlayer=GetOwningPlayer()) -> (ReturnValue=0)
   n4 = int - int(A=@n2, B=@n3)
   > UpdateProgress!(
    execute
    self:Device - Tracker View Model Object Reference=Get ViewModel()
    Value:int=Select!(
     Total:int=@n3
     Remaining:int="Max (Integer)"(A=@n4, B=0)
     Off:int=0
     Index:EShowProgressMode Enum=Get ShowProgressMode(self=@n1) -> (ShowProgressMode=Total)
    ) -> (ReturnValue:int)
    MaxValue:int=@n2
    Percentage:double=SafeDivide(A="To Float (Integer)"(InInt=@n3), B="To Float (Integer)"(InInt=@n2))
   ) -> (then)
  ]}
]
variables [
 Cached Quest Component {type=CreativeQuestComponent, default=null}
 HasReachedCompletionAmount {type=boolean}
 IsLoading {type=boolean}
 LoadedWidget {type=UserWidget, default=null}
 ShowCompleteAnimation {type=boolean}
 ViewModel {type=MVVM_UEFN_Tracker_C, default=null}
 WidgetUpdatedWhileLoading {type=boolean}
]
widgets {
 info {inheritedWidgetCount=5, parentClass=/Script/FortniteUI.FortCreativeQuestHUDWidget, rootWidgetClass=&Overlay, widgetCount=7}
 widgets [
  Overlay {widgetClassPath=&Overlay, widget=Overlay}
  SizeBox_Root {
   widgetClassPath=&SizeBox
   bIsVariable=true
   parent=Overlay
   widget=SizeBox_Root
   properties {bOverride_WidthOverride=true, widthOverride=578}
   slot {class=OverlaySlot}
  }
  Text_CurrentAmount {widgetClassPath=&CommonTextBlock, bInherited=true, widget=null}
  Text_TargetAmount {widgetClassPath=&CommonTextBlock, bInherited=true, widget=null}
  Text_QuestName {widgetClassPath=&CommonTextBlock, bInherited=true, widget=null}
  Text_QuestDetails {widgetClassPath=&CommonTextBlock, bInherited=true, widget=null}
  TutorialHighlight {widgetClassPath=&TutorialHighlight}
 ]
}
