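# Minigame HUD widget that lists creative quest trackers in a vertical box.
# On each in-progress UI display event, it clears the list and looks up the minigame settings machine.
# If none exists, it binds to the volume component's settings-machine-changed event, then unbinds and waits a tick when that fires.
# It creates indexed, tagged trackers up to the machine's HUD maximum, padding all but the last.
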WidgetBlueprint /Game/Creative/Devices/Quests/CreativeQuestTracker_List
graphs [
 EventGraph {nodes [
   n1 = OnMinigameSettingsMachineChanged!() -> (
    OutputDelegate
    then
    MinigameColumeComponent:Fort Minigame Volume Component Object Reference
    NewMinigameSettingsMachine:Fort Minigame Settings Building Object Reference
   ) {is=title K2Node_CustomEvent}
   > Is Valid(InputObject=@n1.NewMinigameSettingsMachine)
   n2 = Create Event()
   > Unbind Event from On Minigame Settings Machine Changed(self:Fort Minigame Volume Component Object Reference=@n1.MinigameColumeComponent, Delegate=@n2) {is=title K2Node_RemoveDelegate}
   > n3 = DelayUntilNextTick!(execute) -> (then)
   n4 = GetOwningPlayer()
   n5 = GetMinigameSettingsMachine!(ContextActor:Actor Object Reference=@n4) -> (ReturnValue:Fort Minigame Settings Building Object Reference)
   n6 = Merge Execution with Data!(
    Option A:Exec=@n13.Is Valid
    Option A Data:Fort Minigame Settings Building Object Reference=@n5
    Option B:Exec=@n3.then
    Option B Data:Fort Minigame Settings Building Object Reference=@n1.NewMinigameSettingsMachine
   ) -> (Then:Exec, Data:Fort Minigame Settings Building Object Reference) {is=title K2Node_MacroInstance}
   > Is Valid(InputObject=@n6)
   n7 = GetMaxHUDTrackers!(self:Fort Minigame Settings Building Object Reference=@n6) -> (ReturnValue:int=0)
   > n8 = For Loop(FirstIndex=1, LastIndex=@n7)
   > n9 = Create Widget!(
    execute:Exec
    Class:User Widget Class Reference=&/CreativeQuestTracker.CreativeQuestTracker_C
    OwningPlayer:Player Controller Object Reference=GetOwningPlayer()
   ) -> (then:Exec, ReturnValue:Creative Quest Tracker Object Reference)
   > SetQuestIndex!(execute, self:Fort Creative Quest HUDWidget Object Reference=@n9, newQuestIndex:int=int - int(A=@n8, B=1) {is=title K2Node_CallFunction}) -> (then)
   > AppendHUDElementTags!(
    execute
    self:Fort HUDElement Widget Object Reference=@n9
    AdditionalTags:"Gameplay Tag Container Structure (by ref)"=MakeGameplayTagContainerFromArray!(GameplayTags:Array of Gameplay Tag Structures=Make Array("[0]"=(TagName="Minigame.UIExtension.QuestTracker"))) -> (ReturnValue:Gameplay Tag Container Structure)
   ) -> (then)
   > Add(
    TargetArray:Array of Creative Quest Tracker Object References=Get Trackers() -> (Trackers:Array of Creative Quest Tracker Object References)
    NewItem:"Creative Quest Tracker Object Reference (by ref)"=@n9
   ) {is=title K2Node_CallArrayFunction}
   > n10 = AddChild(self=Get VerticalBox_TrackerList() -> (VerticalBox_TrackerList:Vertical Box Object Reference), Content=@n9)
   > Branch(Condition="Not Equal (Integer)"!(A:int=@n8, B:int=@n7) -> (ReturnValue:bool) {is=title K2Node_CallFunction})
   > n11 = Cast To VerticalBoxSlot!(execute, Object=@n10) -> (then, CastFailed, AsVertical Box Slot:Vertical Box Slot Object Reference) {is=title K2Node_DynamicCast}
   > SetPadding!(execute, self:Vertical Box Slot Object Reference=@n11, InPadding:Margin Structure=Make Margin(Left=0, Top=0, Right=0, Bottom=12)) -> (then)
   n12 = Event OnDisplayMinigameInProgressUI!() -> (OutputDelegate, then, bDisplay:bool) {is=title K2Node_Event}
   > Parent: OnDisplayMinigameInProgressUI!(execute, bDisplay:bool=@n12.bDisplay) -> (then) {is=title K2Node_CallParentFunction}
   > Clear(TargetArray:Array of Creative Quest Tracker Object References=Get Trackers())
   > ClearChildren(self=Get VerticalBox_TrackerList())
   > n13 = Is Valid(InputObject=@n5)
   n14 = GetMinigameVolumeComponent!(ContextActor:Actor Object Reference=@n4) -> (ReturnValue:Fort Minigame Volume Component Object Reference)
   Is Valid(exec=@n13.Is Not Valid, InputObject=@n14)
   > Bind Event to On Minigame Settings Machine Changed(self:Fort Minigame Volume Component Object Reference=@n14, Delegate=@n2) {is=title K2Node_AddDelegate}
   Event PreConstruct()
   Event Construct()
   Event Tick()
  ]}
 GetVolumeActorForOwningPlayer {nodes [
   Inputs!() -> (exec:Exec)
   n1 = GetOwningPlayer()
   > n2 = Is Valid(InputObject=@n1)
   > n3 = GetMinigameForPlayerState!(
    execute
    PlayerState:Player State Object Reference=Get PlayerState!(self:Controller Object Reference=@n1) -> (PlayerState:Player State Object Reference)
   ) -> (then, ReturnValue:Fort Minigame Object Reference)
   > n4 = Is Valid(InputObject=@n3)
   n5 = Get Volume!(self:Fort Minigame Object Reference=@n3) -> (Volume:Actor Object Reference)
   > n6 = Is Valid(InputObject=@n5)
   > Outputs!(Is Valid:Exec, Is Not Valid:Exec=[@n6.Is Not Valid, @n4.Is Not Valid, @n2.Is Not Valid], VolumeActor:Actor Object Reference=@n5)
  ]}
]
variables [Trackers {type="array<CreativeQuestTracker_C>", default []}]
widgets {
 info {inheritedWidgetCount=1, parentClass=/Script/FortniteUI.MinigameWidgetBase, rootWidgetClass=&SizeBox, widgetCount=3}
 widgets [
  SizeBox_27 {widgetClassPath=&SizeBox, widget=SizeBox_27, properties {widthOverride=300}}
  VerticalBox_TrackerList {
   widgetClassPath=&VerticalBox
   bIsVariable=true
   parent=SizeBox_27
   widget=VerticalBox_TrackerList
   slot {class=SizeBoxSlot, properties {horizontalAlignment=HAlign_Left}}
  }
  TutorialHighlight {widgetClassPath=&TutorialHighlight}
 ]
}
