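; MVVM viewmodel for the UEFN Tracker quest device, holding tracker details and progress for UI.
; UpdateStaticDetails sets name, description, icon, visibilities, completion-animation flag and alpha-stripped color, then calls UpdateProgress with current value and new maximum, passing no percentage, and prints a debug string.
; UpdateProgress is separate because progress changes often.
; It stores value, target and an independently supplied percentage, since trackers counting remaining decrement instead of increment.
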(event Custom|UpdateProgress
  (params (Value :type Integer) (MaxValue :type Integer) (Percentage :type Float%20%28double-precision%29))
  ;[0.1624 0.7049 1 0.11] Updating Progress from MVVM. These can commonly change, so a separate event from the other details, and percentage doesn't necessary mean Value/MaxValue, as value can decrement rather than increment if tracker is tracking remaining.
  (Variables|Default|SetValue (Math|Conversions|ToInteger64(Integer) Value))
  (Variables|Default|SetTargetValue (Math|Conversions|ToInteger64(Integer) MaxValue))
  (Variables|Default|SetPercentage Percentage))

(event Custom|UpdateStaticDetails
  (params (Name :type Text) (Description :type Text) (Color :type Linear%20Color%20Structure) (ProgressVisibility :type ESlateVisibility%20Enum :default "Visible") (MaxValue :type Integer) (ShowCompleteAnimation :type Boolean) (Icon :type Texture%202D%20Object%20Reference) (IconVisibility :type ESlateVisibility%20Enum :default "Visible"))
  (Variables|Default|SetName Name)
  (Variables|Default|SetDescription Description)
  (CallFunction|MakeColor Color)
  (Variables|Default|SetIcon Icon)
  (Variables|Default|SetIconVisibility IconVisibility)
  (Variables|Default|SetProgressVisibility ProgressVisibility)
  (Variables|Default|SetShowCompleteAnimation ShowCompleteAnimation)
  (CallFunction|UpdateProgress (Math|Conversions|ToInteger(Integer64) (Variables|Default|GetValue)) MaxValue)
  (Development|PrintString (Fortnite|GetFirstLocalFortPlayerController 0) (Utilities|String|Append "Tracker Complete Anim - " (Utilities|String|ToString(Boolean) ShowCompleteAnimation))))

(fn MakeColor
  (params (Color_SpecifiedColor_R :type Float%20%28single-precision%29) (Color_SpecifiedColor_G :type Float%20%28single-precision%29) (Color_SpecifiedColor_B :type Float%20%28single-precision%29) (Color_SpecifiedColor_A :type Float%20%28single-precision%29))
  ;[0.2195 0.6584 1] We want our color to not include the alpha.
  (Variables|SetColor_R (Math|Color|MakeColor Color_SpecifiedColor_R Color_SpecifiedColor_G Color_SpecifiedColor_B)))
